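# Conversation database for the Columbus NPC, defining its Star Wars-themed greeting and the services that follow it.
# The entry tag leads to one Speech node that defaults to "Move along." and has contextual lines for a first-ever conversation or specific player character cosmetics.
# The first-encounter line takes its priority from a curve row, while cosmetic lines use zero.
# The speech then links out to the shared NPCLibrary Quests and BasicServices conversations by entry tag.
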ConversationDatabase /NPCLibrary/NPCs/Columbus/Columbus/NPCLibrary_Conversation_Columbus
entry_points [{tag=Conversation.Entry.Columbus, destinations [#1]}]
nodes [
 ConversationEntryPointNode_7 {
  id=#1
  class=ConversationEntryPointNode
  outputs [#2]
  properties {entryTag {tagName=Conversation.Entry.Columbus}}
 }
 Speech_C_4 {
  id=#2
  class=Speech_C
  outputs [#3, #4]
  properties {
   nodeName=""
   generalConfig {
    defaultMessage=Move along.
    contextualMessages [
     {
      message=Identification?
      contextRequirements [{requirement {class=FortControllerRequirement_IsFirstEverConversationWithNPC}}]
      priority {curve {rowName=Tandem.Speech.Priority.FirstEverInteraction}}
     }
     {
      message="Lord Vader approaches!"
      contextRequirements [{requirement {
         class=FortControllerRequirement_HasCID
         properties {allowedCIds [&CID_A_422_Athena_Commando_M_Realm]}
        }}]
      priority {value=0, curve {curveTable=null, rowName=None}}
     }
     {
      message="What are you, some kind of Sith?"
      contextRequirements [{requirement {
         class=FortControllerRequirement_HasCID
         properties {allowedCIds [/BRCosmetics/Athena/Items/Cosmetics/Characters/CID_631_Athena_Commando_M_GalileoKayak_VXLDB]}
        }}]
      priority {value=0, curve {curveTable=null, rowName=None}}
     }
     {
      message=Is that beskar armor?
      contextRequirements [{requirement {
         class=FortControllerRequirement_HasCID
         properties {allowedCIds [/BRCosmetics/Athena/Items/Cosmetics/Characters/CID_960_Athena_Commando_M_Cosmos]}
        }}]
      priority {value=0, curve {curveTable=null, rowName=None}}
     }
     {
      message=You seen that new BT-16?
      requirementMatchPolicy=RequireAny
      contextRequirements [{requirement {
         class=FortControllerRequirement_HasCID
         properties {allowedCIds [
           /BRCosmetics/Athena/Items/Cosmetics/Characters/CID_613_Athena_Commando_M_Columbus_7Y4QE
           /BRCosmetics/Athena/Items/Cosmetics/Characters/CID_635_Athena_Commando_M_GalileoSled_FHJVM
          ]}
        }}]
      priority {value=0, curve {curveTable=null, rowName=None}}
     }
     {
      message=Looking for a bounty?
      requirementMatchPolicy=RequireAny
      contextRequirements [{requirement {
         class=FortControllerRequirement_HasCID
         properties {allowedCIds [
           /BRCosmetics/Athena/Items/Cosmetics/Characters/CID_A_333_Athena_Commando_M_Solstice_C1YP3
           /BRCosmetics/Athena/Items/Cosmetics/Characters/CID_A_338_Athena_Commando_F_Galactic_HN9DO
           /BRCosmetics/Athena/Items/Cosmetics/Characters/CID_A_309_Athena_Commando_M_OrbitTeal_9RBJL
          ]}
        }}]
      priority {value=0, curve {curveTable=null, rowName=None}}
     }
     {
      message="Rebels! Blast 'em!"
      requirementMatchPolicy=RequireAny
      contextRequirements [{requirement {
         class=FortControllerRequirement_HasCID
         properties {allowedCIds [
           /BRCosmetics/Athena/Items/Cosmetics/Characters/Character_DistantEchoCastle
           /BRCosmetics/Athena/Items/Cosmetics/Characters/Character_DistantEchoPro
           /BRCosmetics/Athena/Items/Cosmetics/Characters/Character_DistantEchoPilot
           /BRCosmetics/Athena/Items/Cosmetics/Characters/CID_634_Athena_Commando_F_GalileoRocket_ARVEH
           /BRCosmetics/Athena/Items/Cosmetics/Characters/CID_633_Athena_Commando_M_GalileoFerry_PA3E1
           /BRCosmetics/Athena/Items/Cosmetics/Characters/CID_632_Athena_Commando_F_GalileoZeppelin_SJKPW
           /BRCosmetics/Athena/Items/Cosmetics/Characters/CID_A_411_Athena_Commando_M_Noble
          ]}
        }}]
      priority {value=0, curve {curveTable=null, rowName=None}}
     }
    ]
   }
  }
 }
 ConversationLinkNode_0 {id=#4, class=ConversationLinkNode, properties {remoteEntryTag {tagName=Conversation.Entry.BasicServices}}}
 ConversationLinkNode_1 {id=#3, class=ConversationLinkNode, properties {remoteEntryTag {tagName=Conversation.Entry.NPCLibrary.Quests}}}
]
